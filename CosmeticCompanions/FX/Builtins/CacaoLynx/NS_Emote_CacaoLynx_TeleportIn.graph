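# Niagara system for the CacaoLynx cosmetic companion's teleport-in emote, a short warm-tinted burst anchored on a sampled skeletal mesh.
# Minimal spawns stretched sprites that drop from above over their early life.
# A delayed simultaneous burst then fires Flare (sprite plus blue light), drifting curl-noise BigSparkles, and Shockwave001, an expanding multi-renderer mesh.
# Time-of-day scales brightness, front-end modifiers adjust some values, and Ground_Glow has no renderers.
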NiagaraSystem /CosmeticCompanions/FX/Builtins/CacaoLynx/NS_Emote_CacaoLynx_TeleportIn
dependencies {
 usedDataInterfaces [
  &NiagaraDataInterfaceCurve
  &NiagaraDataInterfaceArrayFloat
  &NiagaraDataInterfaceColorCurve
  &NiagaraDataInterfaceVector2DCurve
  &NiagaraDataInterfaceSkeletalMesh
 ]
 usedDynamicInputs [
  &ReturnNormalizedExecIndex
  &MergeHelper_ColorSwitch
  &FNTimeOfDayScalar
 ]
 usedModules [
  &self:SystemSpawnScript.SystemScriptSource.SystemScriptGraph.NiagaraNodeAssignment_0.SetVariables_#1
  &self:SystemSpawnScript.SystemScriptSource.SystemScriptGraph.NiagaraNodeAssignment_1.SetVariables_#2
  &SystemState
 ]
}
emitters {
 BigSparkles {
  emitterUpdateScript [
   &EmitterState()
   &SpawnBurst_Instantaneous(Spawn Count=5, Spawn Time=0.276, Use Spawn Probability=false, UseLoopCountLimit=false)
  ]
  particleSpawnScript [
   /SetVariables(Particles.VisibilityTag=&RandomRangeInt(Minimum=0, Maximum=1)) {isSetParameters=true}
   &InitializeParticle(
    Color=<rgba 5 1.3942 1.2986 1>
    Lifetime=&RandomRangeFloat(Minimum=0.3, Maximum=0.6)
    Sprite Size Mode=::NewEnumerator3
    Uniform Sprite Size=&Multiply_Float(
     A=&SelectFloatFromArray(Direct Array Index=$Particles.VisibilityTag, Float Selection Array=DataInterface() {propertyValues {FloatData [80,64]}})
     B=&RandomRangeFloat(Minimum=0.5, Maximum=1.3)
    )
    UsePositionOffset=false
    Write Material Random=true
   )
   &CameraOffset(Camera Offset Amount=80)
   &SkeletalMeshLocation(Bone Sampling Mode=::NewEnumerator4, MaskSpawnGroup=false, Sampled Position Offset=<0 0 20>, Skeletal Mesh=$System.Skeletal Mesh Reader)
   &ShapeLocation(
    MaskSpawnGroup=false
    Offset:NiagaraPosition=$Particles.SkeletalMeshLocation.SampledPosition
    Offset Mode=::NewEnumerator0
    Override Seed=false
    Sphere Radius=80
    UseOwnerScale=true
   )
   &AddVelocity(Velocity=&RandomRangeVector(Minimum=<0 -20 0>, Maximum=<1 -60 1>))
  ]
  particleUpdateScript [
   &ParticleState()
   &ScaleSpriteSize(
    Scale Factor=&Multiply_Vector2D_ByFloat(
     Float=1
     Vector2D=&Vector2DFromFloat(Value=&FloatFromCurve(FloatCurve=DataInterface() {propertyValues {Curve {keys [
           {arriveTangent=9.0592, interpMode=RCIM_Cubic, leaveTangent=9.0592, tangentMode=RCTM_User, time=0, value=0}
           {interpMode=RCIM_Cubic, tangentMode=RCTM_User, time=0.2, value=1}
           {tangentMode=RCTM_User, time=1, value=0}
          ]}}}))
    )
    Scale Sprite Size Mode=::NewEnumerator2
   )
   /SetVariables(
    Particles.Color02:LinearColor=&MultiplyLinearColor(
     Linear Color=$Particles.Color
     Scale Factor=&MakeVector4_FromLinearColor(LinearColor=<rgba 0.9441 0.8079 1 1>)
    )
    Particles.SpriteSize02:Vector2f=&Multiply_Vector2(A=$Particles.SpriteSize, B=&Vector2DFromFloat(Value=1))
   ) {isSetParameters=true}
   &ScaleColor(
    Scale Alpha=&Multiply_Float(
     A=&FloatFromCurve(
      FloatCurve=DataInterface() {propertyValues {Curve {keys [
          {interpMode=RCIM_Cubic, tangentMode=RCTM_User, time=0, value=0}
          {arriveTangent=-2.4058, interpMode=RCIM_Cubic, leaveTangent=-2.4058, tangentMode=RCTM_User, time=0.26, value=0.8421}
          {arriveTangent=-1.1775, interpMode=RCIM_Cubic, leaveTangent=-1.1775, time=0.4613, value=0.0645}
          {arriveTangent=-0.1197, interpMode=RCIM_Cubic, leaveTangent=-0.1197, time=0.5898, value=0.4537}
          <at 1 0 RCIM_Cubic>
         ]}}}
      Scale Curve=&Multiply_Float(A=&RandomRangeFloat(Minimum=0.1, Maximum=0.5), B=$System.TODScalar)
     )
     B=&FloatFromCurve(
      FloatCurve=DataInterface() {propertyValues {Curve {keys [{tangentMode=RCTM_User, time=0, value=0}, {tangentMode=RCTM_User, time=1, value=1}]}}}
      CurveIndex=&Multiply_Float(
       A=&MakeFloatFromVector(Vector=Linked() {linkedVariable {name=Particles.Position, type=NiagaraPosition}}, Channel=::NewEnumerator2)
       B=$Engine.System.Age
      )
     )
    )
    ScaleA=true
    ScaleRGB=true
   )
   &CurlNoiseForce(Bias Noise=false, Noise Strength=100, Pan Noise=false)
   &Drag(Drag=&RandomRangeFloat(Minimum=10, Maximum=12), Use Linear Drag=true, Use Rotational Drag=true)
   &SolveForcesAndVelocity(Clamp Velocity=false, Limit Acceleration=false)
  ]
  rendererClasses [&NiagaraSpriteRendererProperties]
  renderers [
   {rendererClass=&NiagaraSpriteRendererProperties}
   {rendererClass=&NiagaraSpriteRendererProperties, rendererIndex=1}
  ]
 }
 Flare {
  emitterUpdateScript [
   &EmitterState()
   &SpawnBurst_Instantaneous(Spawn Time=0.276, Use Spawn Probability=false, UseLoopCountLimit=false)
  ]
  particleSpawnScript [
   &InitializeParticle(
    Color=<rgba 5 0.682 0.4431 1>
    Lifetime=0.4
    Position Offset=<0 0 20>
    Sprite Rotation Mode=::NewEnumerator2
    Sprite Size Mode=::NewEnumerator3
    Uniform Sprite Size=250
    UsePositionOffset=false
    Write Material Random=true
   )
   &DynamicMaterialParameters(
    Index 0 Param 1=&RandomRangeFloat(Minimum=0.4, Maximum=0.6)
    Index 0 Param 2=1
    Index 0 Param 3=1
    Index 0 Param 4=1
    Module.Index0Float0Write=true
    Module.Index0Float1Write=true
    Module.Index0Float2Write=true
    Module.Index0Float3Write=true
   )
   &SkeletalMeshLocation(Bone Sampling Mode=::NewEnumerator4, MaskSpawnGroup=false, Sampled Position Offset=<0 0 20>, Skeletal Mesh=$System.Skeletal Mesh Reader)
   /SetVariables(Particles.LightSpecularScale:NiagaraFloat=&FN_FrontEndFloatModifier(Gameplay=1, Front End=0)) {isSetParameters=true}
  ]
  particleUpdateScript [
   &ParticleState()
   &SolveForcesAndVelocity(Clamp Velocity=false, Limit Acceleration=false)
   &ScaleColor(
    Scale Alpha=&FloatFromCurve(FloatCurve=DataInterface() {propertyValues {Curve {keys [<at 0.8 1>,<at 1 0>]}}}, Scale Curve=$System.TODScalar)
    ScaleA=true
    ScaleRGB=true
   )
   &ScaleSpriteSize(
    Scale Factor=&Vector2DFromFloat(Value=&FloatFromCurve(FloatCurve=DataInterface() {propertyValues {Curve {keys [<at 0 1.5 RCIM_Cubic>,<at 1 0>]}}}))
    Scale Sprite Size Mode=::NewEnumerator2
   )
   &FN_Light_Attributes(
    Color=<rgba 0 3.8263 20 10>
    Create Color Attribute=true
    Create Exponent Attribute=true
    Create Position Attribute=true
    Create Radius Attribute=true
    Exponent=5
    Position=<0 0 20>
    Radius=70
    Scale Color=&FloatFromCurve(FloatCurve=DataInterface() {propertyValues {Curve {keys [<at 0.8 1>,<at 1 0>]}}}, Scale Curve=$System.TODScalar)
   )
  ]
  rendererClasses [&NiagaraSpriteRendererProperties, &NiagaraLightRendererProperties]
  renderers [
   {rendererClass=&NiagaraSpriteRendererProperties}
   {rendererClass=&NiagaraLightRendererProperties, rendererIndex=1}
  ]
 }
 Ground_Glow {rendererClasses [], renderers []}
 Minimal {
  emitterUpdateScript [
   &EmitterState()
   &SpawnBurst_Instantaneous(Spawn Count=2, Use Spawn Probability=false, UseLoopCountLimit=false)
  ]
  particleSpawnScript [
   &FN_SetSpawnGroup(Number of Spawn Groups=2)
   &InitializeParticle(
    Color=&MaskLinearColorBySpawnGroup(Masked Linear Color=<rgba 2 0.8265 0.8125 1>, Passthrough Linear Color=<rgba 3 0.2085 0.2945 0.325>)
    Lifetime=0.409
    Sprite Size Mode=::NewEnumerator3
    Uniform Sprite Size=&MaskFloatBySpawnGroup(Masked Float=100, Passthrough Float=140)
    UsePositionOffset=false
    Write Material Random=true
   )
   &CameraOffset(Camera Offset Amount=&FN_FrontEndFloatModifier(Gameplay=25, Front End=0))
  ]
  particleUpdateScript [
   &ParticleState()
   &ScaleSpriteSize(
    Non-Uniform Curve Sprite Scale=DataInterface() {propertyValues {
      XCurve {keys [
        {interpMode=RCIM_Cubic, leaveTangentWeight=0.05, tangentMode=RCTM_User, time=0, value=0}
        {arriveTangent=0.0367, interpMode=RCIM_Cubic, leaveTangent=0.0367, tangentMode=RCTM_User, time=0.1, value=0.125}
        {arriveTangent=-0.0579, interpMode=RCIM_Cubic, leaveTangent=-0.0579, tangentMode=RCTM_User, time=0.48, value=0.125}
        <at 0.6 0.275 RCIM_Cubic>
       ]}
      YCurve {keys [
        {interpMode=RCIM_Cubic, leaveTangentWeight=0.037, tangentMode=RCTM_User, time=0, value=0}
        {arriveTangent=-0.0209, interpMode=RCIM_Cubic, leaveTangent=-0.0209, tangentMode=RCTM_User, time=0.1, value=1.3}
        <at 0.224 1.3 RCIM_Cubic>
        {arriveTangent=7.8709, interpMode=RCIM_Cubic, leaveTangent=7.8709, tangentMode=RCTM_User, time=0.225, value=1.3}
        {arriveTangent=1.1547, interpMode=RCIM_Cubic, leaveTangent=1.1547, tangentMode=RCTM_User, time=0.3719, value=1.9441}
        {arriveTangent=-4.5964, interpMode=RCIM_Cubic, leaveTangent=-4.5964, time=0.4768, value=1.7895}
        {arriveTangent=-4.3567, interpMode=RCIM_Cubic, leaveTangent=-4.3567, tangentMode=RCTM_User, time=0.512, value=1.3}
        <at 0.6 1.175 RCIM_Cubic>
       ]}
     }}
    Scale Sprite Size Mode=::NewEnumerator3
   )
   &ScaleColor(
    Scale Alpha=&FloatFromCurve(FloatCurve=DataInterface() {propertyValues {Curve {
        defaultValue=1
        keys [
         {arriveTangent=9.0725, interpMode=RCIM_Cubic, leaveTangent=9.0725, tangentMode=RCTM_User, time=0, value=0}
         {arriveTangent=0.0162, interpMode=RCIM_Cubic, leaveTangent=0.0162, tangentMode=RCTM_User, time=0.1408, value=0.9847}
         {arriveTangent=-2.5606, interpMode=RCIM_Cubic, leaveTangent=-2.5606, tangentMode=RCTM_User, time=1, value=0}
        ]
       }}})
    ScaleA=true
    ScaleRGB=true
   )
   &SkeletalMeshLocation(
    Bone Sampling Mode=::NewEnumerator4
    MaskSpawnGroup=false
    Sampled Position Offset=&MakeVector(
     X=0
     Y=0
     Z=&FloatFromCurve(
      FloatCurve=DataInterface() {propertyValues {Curve {keys [<at 0 1 RCIM_Cubic>, {arriveTangent=-5.8953, interpMode=RCIM_Cubic, leaveTangent=-5.8953, tangentMode=RCTM_User, time=0.25, value=0}]}}}
      Scale Curve=200
     )
    )
    Skeletal Mesh=$System.Skeletal Mesh Reader
   )
   /SetVariables(Particles.Position=&Lerp_Vector(
     A=<0 0 150>
     B=Linked() {linkedVariable {name=Particles.SkeletalMeshLocation.SampledPosition, type=NiagaraPosition}}
     Alpha=&FloatFromCurve(FloatCurve=DataInterface() {propertyValues {Curve {keys [<at 0 1>,<at 0.1 0>]}}})
    )) {enabled=false, isSetParameters=true}
  ]
  rendererClasses [&NiagaraSpriteRendererProperties]
  renderers [{rendererClass=&NiagaraSpriteRendererProperties}]
 }
 Shockwave001 {
  emitterUpdateScript [
   &EmitterState()
   &SpawnBurst_Instantaneous(Spawn Time=0.276, Use Spawn Probability=false, UseLoopCountLimit=false)
  ]
  particleSpawnScript [
   &InitializeParticle(
    Color=<rgba 1 1 1 5>
    Lifetime=0.24
    Mesh Scale Mode=::NewEnumerator3
    Mesh Uniform Scale=&Multiply_Float(A=0.5, B=$System.FrontendScale)
    Position Offset=<0 0 20>
    UsePositionOffset=false
    Write Material Random=true
   )
   &InitialMeshOrientation(
    Mesh Orientation Mode=::NewEnumerator7
    Rotation=&RandomRangeVector(Minimum=<0 0 0>, Maximum=<1 0 0>)
    Use Rotation Vector=true
   )
   &SkeletalMeshLocation(Bone Sampling Mode=::NewEnumerator4, MaskSpawnGroup=false, Sampled Position Offset=<0 0 20>, Skeletal Mesh=$System.Skeletal Mesh Reader)
  ]
  particleUpdateScript [
   &ParticleState()
   &DynamicMaterialParameters(
    Index 0 Param 1=1
    Index 0 Param 2=2
    Module.Index0Float0Write=true
    Module.Index0Float1Write=true
    Module.Index0Float2Write=true
    Module.Index0Float3Write=true
   )
   &ScaleColor(
    Scale Alpha=&FloatFromCurve(
     FloatCurve=DataInterface() {propertyValues {Curve {keys [
         {arriveTangent=-1.5846, interpMode=RCIM_Cubic, leaveTangent=-1.5846, tangentMode=RCTM_User, time=0, value=1}
         {arriveTangent=-0.3296, tangentMode=RCTM_User, time=1, value=0}
        ]}}}
     Scale Curve=&Multiply_Float(A=0.4, B=$System.TODScalar)
    )
    Scale RGB=&MakeVectorFromLinearColorRGB(Color=&ColorFromCurve(ColorCurve=DataInterface() {propertyValues {
        AlphaCurve {keys [<at 0 1.5>,<at 1 0>]}
        BlueCurve {keys [<at 0.0389 0.1055>,<at 1 0.1504>]}
        GreenCurve {keys [<at 0.0389 0.1916>,<at 1 0.0547>]}
        RedCurve {keys [<at 0.0389 1>,<at 1 1>]}
       }}))
    ScaleA=true
    ScaleRGB=true
   )
   &ScaleMeshSize(Scale Factor=&VectorFromFloat(Value=&FloatFromCurve(FloatCurve=DataInterface() {propertyValues {Curve {keys [
          {arriveTangent=2.4306, interpMode=RCIM_Cubic, leaveTangent=2.4306, tangentMode=RCTM_User, time=0, value=0}
          {arriveTangent=0.046, interpMode=RCIM_Cubic, leaveTangent=0.046, tangentMode=RCTM_User, time=1, value=1}
         ]}}})))
   &CameraOffset(Camera Offset Amount=15)
   /SetVariables(
    Particles.SecondaryDyn:Vector4f=&MakeVector4(X=2.5, Y=3, Z=0, W=0)
    Particles.SecondarySphereColor:LinearColor=&MakeLinearColorFromVectorAndFloat(
     "Vector (RGB)"=&MakeVectorFromLinearColorRGB(Color=&ColorFromCurve(ColorCurve=DataInterface() {propertyValues {
         AlphaCurve {keys [<at 0 1.5>,<at 1 0>]}
         BlueCurve {keys [<at 0.0389 0.4931>,<at 1 0.2036>]}
         GreenCurve {keys [<at 0.0389 0.7319>,<at 1 0.0694>]}
         RedCurve {keys [<at 0.0389 5>,<at 1 1>]}
        }}))
     "Float (Alpha)"=&FloatFromCurve(
      FloatCurve=DataInterface() {propertyValues {Curve {keys [
          {arriveTangent=-1.5846, interpMode=RCIM_Cubic, leaveTangent=-1.5846, tangentMode=RCTM_User, time=0, value=1}
          {arriveTangent=-0.3296, tangentMode=RCTM_User, time=1, value=0}
         ]}}}
      Scale Curve=&Multiply_Float(A=3, B=$System.TODScalar)
     )
    )
    Particles.ShockwaveColor:LinearColor=&MakeLinearColorFromVectorAndFloat(
     "Vector (RGB)"=&MakeVectorFromLinearColorRGB(Color=<rgba 1 0 0.0469 1>)
     "Float (Alpha)"=&FloatFromCurve(
      FloatCurve=DataInterface() {propertyValues {Curve {keys [
          {arriveTangent=0.0068, interpMode=RCIM_Cubic, leaveTangent=0.0068, tangentMode=RCTM_User, time=0, value=0}
          {arriveTangent=-0.018, interpMode=RCIM_Cubic, leaveTangent=-0.018, tangentMode=RCTM_User, time=0.1, value=1}
          {arriveTangent=0.0005, interpMode=RCIM_Cubic, leaveTangent=0.0005, tangentMode=RCTM_User, time=0.7, value=0}
         ]}}}
      Scale Curve=&Multiply_Float(A=0.5, B=$System.TODScalar)
     )
    )
    Particles.ShockwaveScale:Vector3f=&Multiply_VectorByFloat(Vector=$Particles.Scale, Float=-3.345)
   ) {isSetParameters=true}
  ]
  rendererClasses [&NiagaraMeshRendererProperties]
  renderers [
   {rendererClass=&NiagaraMeshRendererProperties, rendererIndex=0}
   {rendererClass=&NiagaraMeshRendererProperties, rendererIndex=1}
   {rendererClass=&NiagaraMeshRendererProperties, rendererIndex=2}
  ]
 }
}
summary {
 systemName=NS_Emote_CacaoLynx_TeleportIn
 emitters [
  {emitterName=Flare, rendererClasses [&NiagaraSpriteRendererProperties, &NiagaraLightRendererProperties]}
  {emitterName=BigSparkles, rendererClasses [&NiagaraSpriteRendererProperties]}
  {emitterName=Shockwave001, rendererClasses [&NiagaraMeshRendererProperties]}
  {emitterName=Minimal, rendererClasses [&NiagaraSpriteRendererProperties]}
  {emitterName=Ground_Glow, rendererClasses []}
 ]
 userVariables [
  User.bIsFrontEnd {type=NiagaraBool}
  User.bIsFrontEndPreview {type=NiagaraBool}
 ]
}
