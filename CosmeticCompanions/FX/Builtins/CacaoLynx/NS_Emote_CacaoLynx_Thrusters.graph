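# Niagara system for the three-thruster exhaust effect in the CacaoLynx companion emote.
# After a shared delay, Thruster_Start flashes ignition, Thruster_Loop renders flickering mesh-and-sprite flames, Thruster_Smoke trails flipbook smoke, and Minimal adds pulsing lights.
# Spawn group selects each particle's skeletal mesh socket and staggers flame waveform phases.
# Comet_Flame_Mesh and Thruster_Loop001 are disabled, leaving User.Flame_Size and User.Thruster_Height_Mult without effect on active emitters.
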NiagaraSystem /CosmeticCompanions/FX/Builtins/CacaoLynx/NS_Emote_CacaoLynx_Thrusters
dependencies {
 usedDataInterfaces [&NiagaraDataInterfaceCurve, &NiagaraDataInterfaceColorCurve, &NiagaraDataInterfaceVectorCurve]
 usedDynamicInputs [&SimulationPosition, &UniformRangedVector2D]
 usedModules [
  &self:SystemSpawnScript.SystemScriptSource.SystemScriptGraph.NiagaraNodeAssignment_0.SetVariables_#1
  &SystemState
 ]
}
emitters {
 Comet_Flame_Mesh {
  bEnabled=false
  emitterUpdateScript [
   &EmitterState(Life Cycle Mode=::NewEnumerator1, Loop Behavior=::NewEnumerator1, Loop Delay=0.55, Loop Duration Mode=::NewEnumerator1, UseLoopDelay=true)
   &SpawnBurst_Instantaneous(Spawn Count=1, Use Spawn Probability=false)
   &SpawnBurst_Instantaneous(Spawn Count=1, Spawn Group=1, Use Spawn Probability=false) {module=SpawnBurst_Instantaneous002}
   &SpawnBurst_Instantaneous(Spawn Count=1, Spawn Group=2, Use Spawn Probability=false) {module=SpawnBurst_Instantaneous003}
   &SpawnBurst_Instantaneous(Spawn Count=2, Spawn Group=1, Use Spawn Probability=false) {enabled=false, module=SpawnBurst_Instantaneous001}
  ]
  particleSpawnScript [
   &PartitionParticles() {enabled=false}
   &InitializeParticle(
    Lifetime=0.275
    Mass Mode=::NewEnumerator1
    Mesh Scale=<-0.015 0.035 0.03>
    Mesh Scale Mode=::NewEnumerator4
    Position Offset=&MergeHelper_VectorSwitch(0=<0 0 0>, 1=<-20.5 8 -7.5>, 2=<20.5 8 -7.5>, 3=?, Index=&SpawnGroup(), Num=::NewEnumerator1)
    UsePositionOffset=true
    Write Material Random=true
   ) {module=InitializeParticle001}
   &InitialMeshOrientation(Mesh Orientation Mode=::NewEnumerator1, Rotation=<0 0 -0.25>, Use Rotation Vector=true)
   /SetVariables(Particles.Color_2:LinearColor=$System.Thrusters_Color, Particles.Scale_2:Vector3f=<0.035 0.032 0.025>) {isSetParameters=true, module=SetVariables_#2}
   &DynamicMaterialParameters(
    Index 0 Param 1=5
    Index 0 Param 2=5
    Module.Index0Float0Write=true
    Module.Index0Float1Write=true
    Module.Index0Float2Write=true
    Module.Index0Float3Write=true
   )
   /SetVariables(Particles.VisibilityTag=&SpawnGroup()) {enabled=false, isSetParameters=true}
  ]
  particleUpdateScript [
   &ParticleState()
   &ScaleColor(
    Scale Alpha=&Lerp_Float(A=1, B=0.5, Alpha=$Particles.NormalizedAge)
    Scale RGB=&MakeVectorFromLinearColorRGB(Color=&ColorFromCurve(ColorCurve=DataInterface() {propertyValues {
        AlphaCurve {keys []}
        BlueCurve {keys [<at 0 0.1839 RCIM_Cubic>,<at 0.492 0.5>,<at 1 0.875 RCIM_Cubic>]}
        GreenCurve {keys [<at 0 0.6739 RCIM_Cubic>,<at 0.492 0.7348>,<at 1 0.9011 RCIM_Cubic>]}
        RedCurve {keys [<at 0 1 RCIM_Cubic>,<at 0.492 1>,<at 1 1 RCIM_Cubic>]}
       }}))
    ScaleA=true
    ScaleRGB=true
   )
   &ScaleMeshSize(Scale Factor=&MakeVector(
     X=&Multiply_Float(
      A=&Multiply_Float(A=2, B=1.6)
      B=&FloatFromCurve(FloatCurve=DataInterface() {propertyValues {Curve {keys [<at 0 0.75 RCIM_Cubic>, {arriveTangent=-0.25, interpMode=RCIM_Cubic, leaveTangent=-0.25, time=0.5, value=2.5}, <at 1 0.5 RCIM_Cubic>]}}})
     )
     Y=&Multiply_Float(
      A=&Multiply_Float(A=2, B=1.25)
      B=&FloatFromCurve(FloatCurve=DataInterface() {propertyValues {Curve {keys [<at 0.5 1 RCIM_Cubic>,<at 1 1.75 RCIM_Cubic>]}}})
     )
     Z=&Multiply_Float(
      A=&Multiply_Float(A=2, B=1.25)
      B=&FloatFromCurve(FloatCurve=DataInterface() {propertyValues {Curve {keys [<at 0.5 1 RCIM_Cubic>,<at 1 1.75 RCIM_Cubic>]}}})
     )
    ))
   /SetVariables(Particles.Scale_2:Vector3f=&Multiply_Vector(
     A=&Multiply_VectorByFloat(Vector=$Particles.Initial.Scale_2, Float=$User.Thruster_Height_Mult)
     B=&MakeVector(
      X=&FloatFromCurve(
       FloatCurve=DataInterface() {propertyValues {Curve {keys [<at 0 0.75 RCIM_Cubic>, {arriveTangent=-0.25, interpMode=RCIM_Cubic, leaveTangent=-0.25, time=0.5, value=2.5}, <at 1 0.5 RCIM_Cubic>]}}}
       Scale Curve=&Multiply_Float(A=$User.Flame_Size, B=1.5)
      )
      Y=&FloatFromCurve(
       FloatCurve=DataInterface() {propertyValues {Curve {keys [<at 0.5 1 RCIM_Cubic>,<at 1 1.75 RCIM_Cubic>]}}}
       Scale Curve=&Multiply_Float(A=$User.Flame_Size, B=1.5)
      )
      Z=&FloatFromCurve(
       FloatCurve=DataInterface() {propertyValues {Curve {keys [<at 0.5 1 RCIM_Cubic>,<at 1 1.75 RCIM_Cubic>]}}}
       Scale Curve=&Multiply_Float(A=$User.Flame_Size, B=1.5)
      )
     )
    )) {isSetParameters=true, module=SetVariables_#3}
  ]
  rendererClasses [&NiagaraMeshRendererProperties]
  renderers [
   {rendererClass=&NiagaraMeshRendererProperties, rendererIndex=0}
   {rendererClass=&NiagaraMeshRendererProperties, rendererIndex=1}
  ]
 }
 Minimal {
  emitterUpdateScript [
   &EmitterState(Life Cycle Mode=::NewEnumerator1, Loop Behavior=::NewEnumerator1, Loop Delay=0.55, Loop Duration Mode=::NewEnumerator1, UseLoopDelay=true)
   &SpawnBurst_Instantaneous(Use Spawn Probability=false, UseLoopCountLimit=false)
   &SpawnBurst_Instantaneous(Spawn Group=1, Use Spawn Probability=false, UseLoopCountLimit=false) {module=SpawnBurst_Instantaneous002}
   &SpawnBurst_Instantaneous(Spawn Group=2, Use Spawn Probability=false, UseLoopCountLimit=false) {module=SpawnBurst_Instantaneous001}
  ]
  particleSpawnScript [&InitializeParticle(
    Color Mode=::NewEnumerator0
    Position Offset=&MergeHelper_VectorSwitch(0=<0 8.5 -2.5>, 1=<-25.25 17 -8>, 2=<25.25 17 -8>, 3=?, Index=&SpawnGroup(), Num=::NewEnumerator1)
    UsePositionOffset=false
    Write Material Random=true
   )]
  particleUpdateScript [
   &ParticleState(Kill Particles When Lifetime Has Elapsed=false, Let Infinitely Lived Particles Die When Emitter Deactivates=true)
   &SkeletalMeshLocation(
    Bone / Socket Index=&SpawnGroup()
    MaskSpawnGroup=false
    Mesh Sampling Type=::NewEnumerator5
    Sampled Position Offset=<-8.9964 0 0>
    Skeletal Mesh=$System.Skeletal Mesh Reader
    Socket Sampling Mode=::NewEnumerator5
   )
   &FN_Light_Attributes(
    Color=<rgba 1 0.293 0.0711 1>
    Create Color Attribute=true
    Create Exponent Attribute=true
    Create Position Attribute=true
    Create Radius Attribute=true
    Exponent=2.75
    Position:NiagaraPosition=$Particles.SkeletalMeshLocation.SampledPosition
    Radius=&MergeHelper_FloatSwitch(0=60, 1=50, 2=50, 3=?, Index=&SpawnGroup(), Num=::NewEnumerator1)
    Scale Color=&Waveform("Waveform [1]"=::NewEnumerator2, "[1] Frequency"=3, "[1] Amplitude Min/Max"=<0.4 1>, "[1] Scale"=25)
   )
  ]
  rendererClasses [&NiagaraLightRendererProperties]
  renderers [{rendererClass=&NiagaraLightRendererProperties, rendererIndex=0}]
 }
 Thruster_Loop {
  emitterUpdateScript [
   &EmitterState(
    Inactive Response=::NewEnumerator1
    Life Cycle Mode=::NewEnumerator1
    Loop Behavior=::NewEnumerator1
    Loop Delay=0.55
    Loop Duration Mode=::NewEnumerator1
    UseLoopDelay=true
   )
   &SpawnRate(SpawnRate=5, Use Spawn Probability=false)
   &SpawnRate(SpawnGroup=1, SpawnRate=5, Use Spawn Probability=false) {module=SpawnRate001}
   &SpawnRate(SpawnGroup=2, SpawnRate=5, Use Spawn Probability=false) {module=SpawnRate002}
  ]
  particleSpawnScript [
   &InitializeParticle(
    Color=<rgba 1.5 0.375 0.15 5>
    Lifetime=0.35
    Mass Mode=::NewEnumerator1
    Mesh Scale=<-0.0225 0.04 0.04>
    Mesh Scale Mode=::NewEnumerator4
    Position Offset=&MergeHelper_VectorSwitch(0=<0 8.5 -2.5>, 1=<-25.25 17 -8>, 2=<25.25 17 -8>, 3=?, Index=&SpawnGroup(), Num=::NewEnumerator1)
    Sprite Size Mode=::NewEnumerator1
    Uniform Sprite Size Max=20.7778
    Uniform Sprite Size Min=34.0823
    UsePositionOffset=false
    Write Material Random=true
   ) {module=InitializeParticle001}
   /SetVariables(Particles.Color_2:LinearColor=<rgba 20 12.5 2 0.25>, Particles.Scale_2:Vector3f=<0.02 0.035 0.035>) {isSetParameters=true, module=SetVariables_#2}
   &DynamicMaterialParameters(
    Index 0 Param 1=5
    Index 0 Param 2=5
    Module.Index0Float0Write=true
    Module.Index0Float1Write=true
    Module.Index0Float2Write=true
    Module.Index0Float3Write=true
   )
   &InitialMeshOrientation(Mesh Orientation Mode=::NewEnumerator1, Rotation=<0 0 0.5>, Use Rotation Vector=true) {module=InitialMeshOrientation001}
   /SetVariables(Particles.GlowCO=15) {isSetParameters=true, module=SetVariables_#4}
  ]
  particleUpdateScript [
   &ParticleState()
   &SkeletalMeshLocation(
    Bone / Socket Index=&SpawnGroup()
    MaskSpawnGroup=false
    Mesh Sampling Type=::NewEnumerator5
    Skeletal Mesh=$System.Skeletal Mesh Reader
    Socket Sampling Mode=::NewEnumerator5
   )
   &ScaleSpriteSize(Uniform Curve Scale=&Waveform(
     Global Phase=&Add_Float(
      A=$Emitter.Age
      B=&MergeHelper_FloatSwitch(0=0, 1=0.3, 2=0.7, 3=?, Index=&SpawnGroup(), Num=::NewEnumerator1)
     )
     "Waveform [1]"=::NewEnumerator2
     "[1] Frequency"=2
     "[1] Amplitude Min/Max"=<1.2 2>
    ))
   &ScaleColor(
    Scale Alpha=&Lerp_Float(A=1, B=0, Alpha=$Particles.NormalizedAge)
    Scale RGB=&Multiply_VectorByFloat(
     Vector=&MakeVectorFromLinearColorRGB(Color=&ColorFromCurve(ColorCurve=DataInterface() {propertyValues {
         AlphaCurve {keys []}
         BlueCurve {keys [{arriveTangent=-1.2505, interpMode=RCIM_Cubic, time=0, value=0.25}, <at 0.492 0.5>, <at 1 0.875 RCIM_Cubic>]}
         GreenCurve {keys [{arriveTangent=-0.2312, interpMode=RCIM_Cubic, time=0, value=0.4375}, <at 0.492 0.625>, <at 1 0.9062 RCIM_Cubic>]}
         RedCurve {keys [{arriveTangent=2.3442, interpMode=RCIM_Cubic, time=0, value=1}, <at 0.492 1>, <at 1 1 RCIM_Cubic>]}
        }}))
     Float=3
    )
    ScaleA=true
    ScaleRGB=true
   )
   &ScaleMeshSize(Scale Factor=&MakeVector(
     X=&FloatFromCurve(
      FloatCurve=DataInterface() {propertyValues {Curve {keys [<at 0 0.75 RCIM_Cubic>, {arriveTangent=-0.25, interpMode=RCIM_Cubic, leaveTangent=-0.25, time=0.5, value=2.5}, <at 1 0.5 RCIM_Cubic>]}}}
      Scale Curve=&Waveform(
       Global Phase=&Add_Float(
        A=$Emitter.Age
        B=&MergeHelper_FloatSwitch(0=0, 1=0.3, 2=0.7, 3=?, Index=&SpawnGroup(), Num=::NewEnumerator1)
       )
       "Waveform [1]"=::NewEnumerator2
       "[1] Frequency"=3
       "[1] Amplitude Min/Max"=&MergeHelper_Vector2Switch(0=<1.4 2.5>, 1=<0.8 2>, 2=<0.8 2>, 3=?, 4=?, Index=&SpawnGroup(), Num=::NewEnumerator1)
      )
     )
     Y=&FloatFromCurve(FloatCurve=DataInterface() {propertyValues {Curve {keys [<at 0.5 1 RCIM_Cubic>, {arriveTangent=-2, interpMode=RCIM_Cubic, leaveTangent=-2, time=0.75, value=1.75}, <at 1 0 RCIM_Cubic>]}}})
     Z=&FloatFromCurve(FloatCurve=DataInterface() {propertyValues {Curve {keys [<at 0.5 1 RCIM_Cubic>, {arriveTangent=-2, interpMode=RCIM_Cubic, leaveTangent=-2, time=0.75, value=1.75}, <at 1 0 RCIM_Cubic>]}}})
    ))
   /SetVariables(Particles.Scale_2:Vector3f=&Multiply_Vector(
     A=$Particles.Initial.Scale_2
     B=&MakeVector(
      X=&FloatFromCurve(
       FloatCurve=DataInterface() {propertyValues {Curve {keys [<at 0 0 RCIM_Cubic>,<at 0.5 2.5 RCIM_Cubic>,<at 1 0 RCIM_Cubic>]}}}
       Scale Curve=&Waveform(
        Global Phase=&Add_Float(
         A=$Emitter.Age
         B=&MergeHelper_FloatSwitch(0=0, 1=0.3, 2=0.7, 3=?, Index=&SpawnGroup(), Num=::NewEnumerator1)
        )
        "Waveform [1]"=::NewEnumerator2
        "[1] Frequency"=3
        "[1] Amplitude Min/Max"=&MergeHelper_Vector2Switch(0=<1.5 2.2>, 1=<1 1.7>, 2=<1 1.7>, 3=?, 4=?, Index=&SpawnGroup(), Num=::NewEnumerator1)
       )
      )
      Y=&FloatFromCurve(FloatCurve=DataInterface() {propertyValues {Curve {keys [<at 0.25 1 RCIM_Cubic>, {arriveTangent=-1.3333, interpMode=RCIM_Cubic, leaveTangent=-1.3333, time=0.5, value=1.75}, <at 1 0 RCIM_Cubic>]}}})
      Z=&FloatFromCurve(FloatCurve=DataInterface() {propertyValues {Curve {keys [<at 0.25 1 RCIM_Cubic>, {arriveTangent=-1.3333, interpMode=RCIM_Cubic, leaveTangent=-1.3333, time=0.5, value=1.75}, <at 1 0 RCIM_Cubic>]}}})
     )
    )) {isSetParameters=true, module=SetVariables_#3}
   &FN_Light_Attributes(
    Color=&ColorFromCurve(ColorCurve=DataInterface() {propertyValues {
       AlphaCurve {keys [<at 0 1 RCIM_Cubic>,<at 1 0 RCIM_Cubic>]}
       BlueCurve {keys [<at 0 0 RCIM_Cubic>,<at 1 0.0003 RCIM_Cubic>]}
       GreenCurve {keys [<at 0 1.0901 RCIM_Cubic>,<at 1 0.0351 RCIM_Cubic>]}
       RedCurve {keys [<at 0 12 RCIM_Cubic>,<at 1 0.8229 RCIM_Cubic>]}
      }})
    Create Color Attribute=true
    Create Exponent Attribute=true
    Create Radius Attribute=true
    Exponent=4
    Radius=30
    Scale Color=10
   )
   /SetVariables(Particles.YZ:NiagaraFloat=&FloatFromCurve(
     FloatCurve=DataInterface() {propertyValues {Curve {keys [<at 0 0.75 RCIM_Cubic>, {arriveTangent=-0.25, interpMode=RCIM_Cubic, leaveTangent=-0.25, time=0.5, value=2.5}, <at 1 0.5 RCIM_Cubic>]}}}
     Scale Curve=&Waveform(
      Global Phase=&Add_Float(
       A=$Emitter.Age
       B=&MergeHelper_FloatSwitch(0=0, 1=0.3, 2=0.7, 3=?, Index=&SpawnGroup(), Num=::NewEnumerator1)
      )
      "Waveform [1]"=::NewEnumerator2
      "[1] Frequency"=3
      "[1] Amplitude Min/Max"=&MergeHelper_Vector2Switch(0=<0 0.4>, 1=<0 0.325>, 2=<0 0.325>, 3=?, 4=?, Index=&SpawnGroup(), Num=::NewEnumerator1)
     )
    )) {isSetParameters=true, module=SetVariables_#5}
   /SetVariables(
    Particles.Color_3:LinearColor=<rgba 5 2.5557 0.7286 1>
    Particles.Position_3:NiagaraPosition=&MergeHelper_VectorSwitch(
     0=&AddVectorToPosition(Position=$Particles.Position, Vector=<0 0 0>)
     1=&AddVectorToPosition(Position=$Particles.Position, Vector=<0 0 0>)
     2=&AddVectorToPosition(Position=$Particles.Position, Vector=<0 0 0>)
     3=?
     Index=&SpawnGroup()
     Num=::NewEnumerator1
    )
    Particles.Scale_3:Vector3f=&MakeVector(
     X=&Multiply_Float(A=$Particles.YZ, B=0.15)
     Y=&Multiply_Float(A=$Particles.YZ, B=0.18)
     Z=&Multiply_Float(A=$Particles.YZ, B=0.18)
    )
   ) {isSetParameters=true, module=SetVariables_#6}
   /SetVariables(Particles.Color_2:LinearColor=&MakeLinearColorFromVectorAndFloat(
     "Vector (RGB)"=&Multiply_VectorByFloat(
      Vector=&MakeVectorFromLinearColorRGB(Color=&ColorFromCurve(ColorCurve=DataInterface() {propertyValues {
          AlphaCurve {keys []}
          BlueCurve {keys [{arriveTangent=-1.2505, interpMode=RCIM_Cubic, time=0, value=0.25}, <at 0.492 0.5>, <at 1 0.875 RCIM_Cubic>]}
          GreenCurve {keys [{arriveTangent=-0.2312, interpMode=RCIM_Cubic, time=0, value=0.6413}, <at 0.492 0.7348>, <at 1 0.9076 RCIM_Cubic>]}
          RedCurve {keys [{arriveTangent=2.3442, interpMode=RCIM_Cubic, time=0, value=1}, <at 0.492 1>, <at 1 1 RCIM_Cubic>]}
         }}))
      Float=3
     )
     "Float (Alpha)"=1
    )) {isSetParameters=true, module=SetVariables_#7}
  ]
  rendererClasses [&NiagaraMeshRendererProperties, &NiagaraSpriteRendererProperties]
  renderers [
   {rendererClass=&NiagaraMeshRendererProperties, rendererIndex=0}
   {rendererClass=&NiagaraMeshRendererProperties, rendererIndex=1}
   {rendererClass=&NiagaraSpriteRendererProperties, rendererIndex=2}
   {rendererClass=&NiagaraMeshRendererProperties, rendererIndex=3}
  ]
 }
 Thruster_Loop001 {
  bEnabled=false
  emitterUpdateScript [
   &EmitterState(
    Inactive Response=::NewEnumerator1
    Life Cycle Mode=::NewEnumerator1
    Loop Behavior=::NewEnumerator1
    Loop Delay=0.55
    Loop Duration Mode=::NewEnumerator1
    UseLoopDelay=true
   )
   &SpawnRate(SpawnRate=5, Use Spawn Probability=false)
   &SpawnRate(SpawnGroup=1, SpawnRate=5, Use Spawn Probability=false) {enabled=false, module=SpawnRate001}
   &SpawnRate(SpawnGroup=2, SpawnRate=5, Use Spawn Probability=false) {enabled=false, module=SpawnRate002}
  ]
  particleSpawnScript [
   &InitializeParticle(
    Color=<rgba 1.5 0.375 0.15 5>
    Lifetime=0.35
    Mass Mode=::NewEnumerator1
    Mesh Scale=<-0.0225 0.04 0.04>
    Mesh Scale Mode=::NewEnumerator4
    Position Offset=&MergeHelper_VectorSwitch(0=<0 0 0>, 1=<-20.5 10.5 -7.5>, 2=<20.5 10.5 -7.5>, 3=?, Index=&SpawnGroup(), Num=::NewEnumerator1)
    Sprite Size Mode=::NewEnumerator1
    Uniform Sprite Size Max=20.7778
    Uniform Sprite Size Min=34.0823
    UsePositionOffset=true
    Write Material Random=true
   ) {module=InitializeParticle001}
   /SetVariables(Particles.Color_2:LinearColor=<rgba 20 12.5 2 0.25>, Particles.Scale_2:Vector3f=<0.02 0.035 0.035>) {isSetParameters=true, module=SetVariables_#2}
   &DynamicMaterialParameters(
    Index 0 Param 1=5
    Index 0 Param 2=5
    Module.Index0Float0Write=true
    Module.Index0Float1Write=true
    Module.Index0Float2Write=true
    Module.Index0Float3Write=true
   )
   &InitialMeshOrientation(Mesh Orientation Mode=::NewEnumerator1, Rotation=<0 0 -0.25>, Use Rotation Vector=true) {module=InitialMeshOrientation001}
   /SetVariables(Particles.GlowCO=15) {isSetParameters=true, module=SetVariables_#4}
  ]
  particleUpdateScript [
   &ParticleState()
   &ScaleSpriteSize(Uniform Curve Scale=&Waveform(
     Global Phase=&Add_Float(
      A=$Emitter.Age
      B=&MergeHelper_FloatSwitch(0=0, 1=0.3, 2=0.7, 3=?, Index=&SpawnGroup(), Num=::NewEnumerator1)
     )
     "Waveform [1]"=::NewEnumerator2
     "[1] Frequency"=2
     "[1] Amplitude Min/Max"=<1.2 2>
    ))
   &ScaleColor(
    Scale Alpha=&Lerp_Float(A=1, B=0, Alpha=$Particles.NormalizedAge)
    Scale RGB=&Multiply_VectorByFloat(
     Vector=&MakeVectorFromLinearColorRGB(Color=&ColorFromCurve(ColorCurve=DataInterface() {propertyValues {
         AlphaCurve {keys []}
         BlueCurve {keys [{arriveTangent=-1.2505, interpMode=RCIM_Cubic, time=0, value=0.25}, <at 0.492 0.5>, <at 1 0.875 RCIM_Cubic>]}
         GreenCurve {keys [{arriveTangent=-0.2312, interpMode=RCIM_Cubic, time=0, value=0.4375}, <at 0.492 0.625>, <at 1 0.9062 RCIM_Cubic>]}
         RedCurve {keys [{arriveTangent=2.3442, interpMode=RCIM_Cubic, time=0, value=1}, <at 0.492 1>, <at 1 1 RCIM_Cubic>]}
        }}))
     Float=3
    )
    ScaleA=true
    ScaleRGB=true
   )
   &ScaleMeshSize(Scale Factor=&MakeVector(
     X=&FloatFromCurve(
      FloatCurve=DataInterface() {propertyValues {Curve {keys [<at 0 0.75 RCIM_Cubic>, {arriveTangent=-0.25, interpMode=RCIM_Cubic, leaveTangent=-0.25, time=0.5, value=2.5}, <at 1 0.5 RCIM_Cubic>]}}}
      Scale Curve=&Waveform(
       Global Phase=&Add_Float(
        A=$Emitter.Age
        B=&MergeHelper_FloatSwitch(0=0, 1=0.3, 2=0.7, 3=?, Index=&SpawnGroup(), Num=::NewEnumerator1)
       )
       "Waveform [1]"=::NewEnumerator2
       "[1] Frequency"=2
       "[1] Amplitude Min/Max"=&MergeHelper_Vector2Switch(0=<0.7 3>, 1=<0.425 2.2>, 2=<0.425 2.2>, 3=?, 4=?, Index=&SpawnGroup(), Num=::NewEnumerator1)
      )
     )
     Y=&FloatFromCurve(FloatCurve=DataInterface() {propertyValues {Curve {keys [<at 0.5 1 RCIM_Cubic>, {arriveTangent=-2, interpMode=RCIM_Cubic, leaveTangent=-2, time=0.75, value=1.75}, <at 1 0 RCIM_Cubic>]}}})
     Z=&FloatFromCurve(FloatCurve=DataInterface() {propertyValues {Curve {keys [<at 0.5 1 RCIM_Cubic>, {arriveTangent=-2, interpMode=RCIM_Cubic, leaveTangent=-2, time=0.75, value=1.75}, <at 1 0 RCIM_Cubic>]}}})
    ))
   /SetVariables(Particles.Scale_2:Vector3f=&Multiply_Vector(
     A=$Particles.Initial.Scale_2
     B=&MakeVector(
      X=&FloatFromCurve(
       FloatCurve=DataInterface() {propertyValues {Curve {keys [<at 0 0.75 RCIM_Cubic>, {arriveTangent=0.25, interpMode=RCIM_Cubic, leaveTangent=0.25, time=0.5, value=2.5}, <at 1 1 RCIM_Cubic>]}}}
       Scale Curve=&Waveform(
        Global Phase=&Add_Float(
         A=$Emitter.Age
         B=&MergeHelper_FloatSwitch(0=0, 1=0.3, 2=0.7, 3=?, Index=&SpawnGroup(), Num=::NewEnumerator1)
        )
        "Waveform [1]"=::NewEnumerator2
        "[1] Frequency"=1
        "[1] Compound Frequency"=1
        "[1] Amplitude Min/Max"=&MergeHelper_Vector2Switch(0=<1 3.2>, 1=<0.425 2.8>, 2=<0.425 2.8>, 3=?, 4=?, Index=&SpawnGroup(), Num=::NewEnumerator1)
       )
      )
      Y=&FloatFromCurve(FloatCurve=DataInterface() {propertyValues {Curve {keys [<at 0.25 1 RCIM_Cubic>, {arriveTangent=-1.3333, interpMode=RCIM_Cubic, leaveTangent=-1.3333, time=0.5, value=1.75}, <at 1 0 RCIM_Cubic>]}}})
      Z=&FloatFromCurve(FloatCurve=DataInterface() {propertyValues {Curve {keys [<at 0.25 1 RCIM_Cubic>, {arriveTangent=-1.3333, interpMode=RCIM_Cubic, leaveTangent=-1.3333, time=0.5, value=1.75}, <at 1 0 RCIM_Cubic>]}}})
     )
    )) {isSetParameters=true, module=SetVariables_#3}
   &FN_Light_Attributes(
    Color=&ColorFromCurve(ColorCurve=DataInterface() {propertyValues {
       AlphaCurve {keys [<at 0 1 RCIM_Cubic>,<at 1 0 RCIM_Cubic>]}
       BlueCurve {keys [<at 0 0 RCIM_Cubic>,<at 1 0.0003 RCIM_Cubic>]}
       GreenCurve {keys [<at 0 1.0901 RCIM_Cubic>,<at 1 0.0351 RCIM_Cubic>]}
       RedCurve {keys [<at 0 12 RCIM_Cubic>,<at 1 0.8229 RCIM_Cubic>]}
      }})
    Create Color Attribute=true
    Create Exponent Attribute=true
    Create Radius Attribute=true
    Exponent=4
    Radius=30
    Scale Color=10
   )
   /SetVariables(Particles.YZ:NiagaraFloat=&FloatFromCurve(
     FloatCurve=DataInterface() {propertyValues {Curve {keys [<at 0 0.75 RCIM_Cubic>, {arriveTangent=-0.25, interpMode=RCIM_Cubic, leaveTangent=-0.25, time=0.5, value=2.5}, <at 1 0.5 RCIM_Cubic>]}}}
     Scale Curve=&Waveform(
      Global Phase=&Add_Float(
       A=$Emitter.Age
       B=&MergeHelper_FloatSwitch(0=0, 1=0.3, 2=0.7, 3=?, Index=&SpawnGroup(), Num=::NewEnumerator1)
      )
      "Waveform [1]"=::NewEnumerator2
      "[1] Frequency"=2
      "[1] Amplitude Min/Max"=&MergeHelper_Vector2Switch(0=<0.075 0.4>, 1=<0.1 0.325>, 2=<0.1 0.325>, 3=?, 4=?, Index=&SpawnGroup(), Num=::NewEnumerator1)
     )
    )) {isSetParameters=true, module=SetVariables_#5}
   /SetVariables(
    Particles.Color_3:LinearColor=<rgba 5 2.5557 0.7286 1>
    Particles.Position_3:NiagaraPosition=&MergeHelper_VectorSwitch(
     0=&AddVectorToPosition(Position=$Particles.Position, Vector=<0 -1.5 0>)
     1=&AddVectorToPosition(Position=$Particles.Position, Vector=<0 -1.5 -1>)
     2=&AddVectorToPosition(Position=$Particles.Position, Vector=<0 -1.5 -1>)
     3=?
     Index=&SpawnGroup()
     Num=::NewEnumerator1
    )
    Particles.Scale_3:Vector3f=&MakeVector(
     X=&Multiply_Float(A=$Particles.YZ, B=0.18)
     Y=&Multiply_Float(A=$Particles.YZ, B=0.23)
     Z=&Multiply_Float(A=$Particles.YZ, B=0.23)
    )
   ) {isSetParameters=true, module=SetVariables_#6}
   /SetVariables(Particles.Color_2:LinearColor=&MakeLinearColorFromVectorAndFloat(
     "Vector (RGB)"=&Multiply_VectorByFloat(
      Vector=&MakeVectorFromLinearColorRGB(Color=&ColorFromCurve(ColorCurve=DataInterface() {propertyValues {
          AlphaCurve {keys []}
          BlueCurve {keys [{arriveTangent=-1.2505, interpMode=RCIM_Cubic, time=0, value=0.25}, <at 0.492 0.5>, <at 1 0.875 RCIM_Cubic>]}
          GreenCurve {keys [{arriveTangent=-0.2312, interpMode=RCIM_Cubic, time=0, value=0.6413}, <at 0.492 0.7348>, <at 1 0.9076 RCIM_Cubic>]}
          RedCurve {keys [{arriveTangent=2.3442, interpMode=RCIM_Cubic, time=0, value=1}, <at 0.492 1>, <at 1 1 RCIM_Cubic>]}
         }}))
      Float=3
     )
     "Float (Alpha)"=1
    )) {isSetParameters=true, module=SetVariables_#7}
  ]
  rendererClasses [&NiagaraMeshRendererProperties, &NiagaraSpriteRendererProperties]
  renderers [
   {rendererClass=&NiagaraMeshRendererProperties, rendererIndex=0}
   {rendererClass=&NiagaraMeshRendererProperties, rendererIndex=1}
   {rendererClass=&NiagaraSpriteRendererProperties, rendererIndex=2}
   {rendererClass=&NiagaraMeshRendererProperties, rendererIndex=3}
  ]
 }
 Thruster_Smoke {
  emitterUpdateScript [
   &EmitterState(
    Inactive Response=::NewEnumerator1
    Life Cycle Mode=::NewEnumerator1
    Loop Behavior=::NewEnumerator1
    Loop Delay=0.55
    Loop Duration Mode=::NewEnumerator1
    UseLoopDelay=true
   )
   &SpawnRate(SpawnRate=20, Use Spawn Probability=false)
   &SpawnRate(SpawnGroup=1, SpawnRate=20, Use Spawn Probability=false) {module=SpawnRate001}
   &SpawnRate(SpawnGroup=2, SpawnRate=20, Use Spawn Probability=false) {module=SpawnRate002}
  ]
  particleSpawnScript [
   &InitializeParticle(
    Color=<rgba 1 1 1 0.0625>
    Lifetime=&UniformRangedFloat(Minimum=0.55, Maximum=0.75)
    Position Offset=&MergeHelper_VectorSwitch(0=<0 8.5 -2.5>, 1=<-25.25 17 -8>, 2=<25.25 17 -8>, 3=?, Index=&SpawnGroup(), Num=::NewEnumerator1)
    Sprite Rotation Mode=::NewEnumerator2
    Sprite Size Mode=::NewEnumerator3
    Uniform Sprite Size=&MakeFloatFromVector2d(Vector 2D=&Vector2DFromFloat(Value=&UniformRangedFloat(Minimum=18, Maximum=24)))
    UsePositionOffset=false
    Write Material Random=true
   ) {module=InitializeParticle001}
   &DynamicMaterialParameters(
    Index 0 Param 2=1
    Index 0 Param 4=1
    Module.Index0Float0Write=true
    Module.Index0Float1Write=true
    Module.Index0Float2Write=true
    Module.Index0Float3Write=true
   )
   &AddVelocity(
    Cone Angle=5
    Cone Axis=<0 -1 -0.75>
    Module.UseConeFalloff=false
    Module.UseVelDistribution=false
    Override Seed=false
    Velocity Mode=::NewEnumerator2
    Velocity Speed=&RandomRangeFloat(Minimum=80, Maximum=150)
   ) {module=AddVelocity001}
  ]
  particleUpdateScript [
   &ParticleState()
   &SkeletalMeshLocation(
    Bone / Socket Index=&SpawnGroup()
    MaskSpawnGroup=false
    Mesh Sampling Type=::NewEnumerator5
    Position Sampling=::NewEnumerator1
    Sampled Position Offset=<-3.3195 0 0>
    Skeletal Mesh=$System.Skeletal Mesh Reader
    Socket Sampling Mode=::NewEnumerator5
   )
   &Drag(Drag=&RandomRangeFloat(Minimum=0.8, Maximum=1.2), Use Linear Drag=true, Use Rotational Drag=false)
   &SolveForcesAndVelocity(Clamp Velocity=false, Limit Acceleration=false)
   &DynamicMaterialParameters(
    Index 0 Param 1=&Lerp_Float(A=1, B=-1, Alpha=$Particles.NormalizedAge)
    Index 0 Param 2=1
    Index 0 Param 4=1
    Module.Index0Float0Write=true
    Module.Index0Float1Write=true
    Module.Index0Float2Write=true
    Module.Index0Float3Write=true
   ) {module=DynamicMaterialParameters001}
   /SetVariables(Particles.SubImageIndex=&FloatFromCurve(FloatCurve=DataInterface() {propertyValues {Curve {keys [{arriveTangent=105.1181, interpMode=RCIM_Cubic, leaveTangent=105.1181, tangentMode=RCTM_User, time=0, value=0}, <at 1 63 RCIM_Cubic>]}}})) {isSetParameters=true}
   &ScaleColor(
    Scale Alpha=&FloatFromCurve(FloatCurve=DataInterface() {propertyValues {Curve {keys [<at 0 0 RCIM_Cubic>,<at 0.1 1 RCIM_Cubic>,<at 1 0 RCIM_Cubic>]}}})
    Scale RGB=&VectorFromFloat(Value=&FloatFromCurve(FloatCurve=DataInterface() {propertyValues {Curve {keys [<at 0 5 RCIM_Cubic>,<at 0.5 1 RCIM_Cubic>]}}}))
    ScaleA=true
    ScaleRGB=true
   )
   &ScaleColor(
    Scale RGB=&MakeVectorFromLinearColorRGB(Color=&ColorFromCurve(ColorCurve=DataInterface() {propertyValues {
        AlphaCurve {keys []}
        BlueCurve {keys [<at 0 0.25 RCIM_Cubic>,<at 0.6182 0.125 RCIM_Cubic>]}
        GreenCurve {keys [<at 0 0.4375 RCIM_Cubic>,<at 0.6182 0.125 RCIM_Cubic>]}
        RedCurve {keys [<at 0 1 RCIM_Cubic>,<at 0.6182 0.125 RCIM_Cubic>]}
       }}))
    ScaleA=true
    ScaleRGB=true
   ) {module=ScaleColor001}
   &ScaleSpriteSize(Scale Factor=&Vector2DFromFloat(Value=&FloatFromCurve(FloatCurve=DataInterface() {propertyValues {Curve {keys [{arriveTangent=6.2193, interpMode=RCIM_Cubic, leaveTangent=6.2193, tangentMode=RCTM_User, time=0, value=0.2}, <at 1 2 RCIM_Cubic>]}}})))
  ]
  rendererClasses [&NiagaraSpriteRendererProperties]
  renderers [{rendererClass=&NiagaraSpriteRendererProperties}]
 }
 Thruster_Start {
  emitterUpdateScript [
   &EmitterState(Life Cycle Mode=::NewEnumerator1, Loop Behavior=::NewEnumerator1, Loop Delay=0.55, Loop Duration=1, UseLoopDelay=true)
   &SpawnBurst_Instantaneous(Spawn Count=3, Use Spawn Probability=false)
  ]
  particleSpawnScript [
   &FN_SetSpawnGroup(Number of Spawn Groups=3)
   &InitializeParticle(
    Color=<rgba 1.5 0.55 0 0.1>
    Lifetime=0.2
    Mass Mode=::NewEnumerator1
    Mesh Scale=<0.333 1 1>
    Mesh Scale Mode=::NewEnumerator4
    Position Offset=&MergeHelper_VectorSwitch(0=<0 8.5 -2.5>, 1=<-25.25 17 -8>, 2=<25.25 17 -8>, 3=?, Index=&SpawnGroup(), Num=::NewEnumerator1)
    Sprite Size Max=<50 1>
    Sprite Size Min=<100 4>
    Sprite Size Mode=::NewEnumerator2
    UsePositionOffset=false
    Write Material Random=true
   ) {module=InitializeParticle001}
   /SetVariables(
    Particles.CameraOffset_2=10
    Particles.Color_2:LinearColor=<rgba 1 0.3667 0 1>
    Particles.DynamicMaterialParameter_2:Vector4f=<1 1 1 1>
    Particles.SpriteSize_2:Vector2f=<100 100>
   ) {isSetParameters=true}
   /SetVariables(
    Particles.Color_3:LinearColor=<rgba 5 5 5 1>
    Particles.DynamicMaterialParameter_3:Vector4f=&MakeVector4(X=0.5, Y=1, Z=&UniformRangedFloat(Minimum=0.25, Maximum=0.35), W=1)
    Particles.SpriteSize_3:Vector2f=&Vector2DFromFloat(Value=&UniformRangedFloat(Minimum=50, Maximum=75))
   ) {isSetParameters=true}
   /SetVariables(
    Particles.Color_4:LinearColor=<rgba 1.5 0.7875 0.375 0.1>
    Particles.DynamicMaterialParameter_4:Vector4f=&MakeVector4(X=1, Y=5, Z=20, W=1)
   ) {isSetParameters=true}
   &CameraOffset(Camera Offset Amount=2)
   &DynamicMaterialParameters(
    Index 0 Param 1=5
    Module.Index0Float0Write=true
    Module.Index0Float1Write=true
    Module.Index0Float2Write=true
    Module.Index0Float3Write=true
   )
   &InitialMeshOrientation(Mesh Orientation Mode=::NewEnumerator1, Rotation=<0 0 0.5>, Use Rotation Vector=true) {module=InitialMeshOrientation001}
  ]
  particleUpdateScript [
   &ParticleState()
   &SkeletalMeshLocation(
    Bone / Socket Index=&SpawnGroup()
    MaskSpawnGroup=false
    Mesh Sampling Type=::NewEnumerator5
    Skeletal Mesh=$System.Skeletal Mesh Reader
    Socket Sampling Mode=::NewEnumerator5
   )
   &ScaleSpriteSize(Scale Factor=&MakeVector2D(X=&FloatFromCurve(FloatCurve=DataInterface() {propertyValues {Curve {keys [<at 0 0>,<at 1 2>]}}}), Y=1))
   /SetVariables(Particles.SpriteSize_3:Vector2f=&Multiply_Vector2(
     A=$Particles.Initial.SpriteSize_3
     B=&Vector2DFromFloat(Value=&FloatFromCurve(FloatCurve=DataInterface() {propertyValues {Curve {keys [<at 0 0>,<at 1 1>]}}}))
    )) {isSetParameters=true}
   &ScaleColor(
    Scale Alpha=&FloatFromCurve(FloatCurve=DataInterface() {propertyValues {Curve {keys [<at 0 1>,<at 1 0>]}}})
    Scale RGB=&MakeVectorFromLinearColorRGB(Color=&ColorFromCurve(ColorCurve=DataInterface() {propertyValues {
        AlphaCurve {keys []}
        BlueCurve {keys [<at 0 0.25 RCIM_Cubic>,<at 0.5 0.5>,<at 1 0.875 RCIM_Cubic>]}
        GreenCurve {keys [<at 0 0.525 RCIM_Cubic>,<at 0.5 0.6833>,<at 1 0.9208 RCIM_Cubic>]}
        RedCurve {keys [<at 0 1 RCIM_Cubic>,<at 0.5 1>,<at 1 1 RCIM_Cubic>]}
       }}))
    ScaleA=true
    ScaleRGB=true
   )
   &ScaleColor(
    Scale RGB=&VectorFromFloat(Value=&FloatFromCurve(FloatCurve=DataInterface() {propertyValues {Curve {keys [
          <at 0 100 RCIM_Cubic>
          {arriveTangent=-33.2952, interpMode=RCIM_Cubic, leaveTangent=-33.2924, tangentMode=RCTM_User, time=0.25, value=10}
          <at 1 0 RCIM_Cubic>
         ]}}}))
    ScaleA=true
    ScaleRGB=true
   ) {module=ScaleColor001}
   /SetVariables(
    Particles.Color_2:LinearColor=&MultiplyLinearColor(
     Linear Color=$Particles.Initial.Color_2
     Scale Factor=&MakeVector4_FromLinearColor(LinearColor=&MakeLinearColorFromVectorAndFloat(
       "Vector (RGB)"=&MakeVectorFromLinearColorRGB(Color=&ColorFromCurve(ColorCurve=DataInterface() {propertyValues {
           AlphaCurve {keys []}
           BlueCurve {keys [<at 0 0.25 RCIM_Cubic>,<at 0.5 0.5>,<at 1 0.875 RCIM_Cubic>]}
           GreenCurve {keys [<at 0 0.525 RCIM_Cubic>,<at 0.5 0.6833>,<at 1 0.9208 RCIM_Cubic>]}
           RedCurve {keys [<at 0 1 RCIM_Cubic>,<at 0.5 1>,<at 1 1 RCIM_Cubic>]}
          }}))
       "Float (Alpha)"=&FloatFromCurve(FloatCurve=DataInterface() {propertyValues {Curve {keys [<at 0 1>,<at 1 0>]}}})
      ))
    )
    Particles.Color_3:LinearColor=&MultiplyLinearColor(
     Linear Color=$Particles.Initial.Color_3
     Scale Factor=&MakeVector4_FromLinearColor(LinearColor=&MakeLinearColorFromVectorAndFloat(
       "Vector (RGB)"=&MakeVectorFromLinearColorRGB(Color=&ColorFromCurve(ColorCurve=DataInterface() {propertyValues {
           AlphaCurve {keys []}
           BlueCurve {keys [<at 0 0.25 RCIM_Cubic>,<at 0.5 0.5>,<at 1 0.875 RCIM_Cubic>]}
           GreenCurve {keys [<at 0 0.525 RCIM_Cubic>,<at 0.5 0.6833>,<at 1 0.9208 RCIM_Cubic>]}
           RedCurve {keys [<at 0 1 RCIM_Cubic>,<at 0.5 1>,<at 1 1 RCIM_Cubic>]}
          }}))
       "Float (Alpha)"=&FloatFromCurve(FloatCurve=DataInterface() {propertyValues {Curve {keys [<at 0 1>,<at 1 0>]}}})
      ))
    )
    Particles.Color_4:LinearColor=&MultiplyLinearColor(
     Linear Color=$Particles.Initial.Color_4
     Scale Factor=&MakeVector4_FromLinearColor(LinearColor=&MakeLinearColorFromVectorAndFloat(
       "Vector (RGB)"=&MakeVectorFromLinearColorRGB(Color=&ColorFromCurve(ColorCurve=DataInterface() {propertyValues {
           AlphaCurve {keys []}
           BlueCurve {keys [<at 0 0.25 RCIM_Cubic>,<at 0.5 0.5>,<at 1 0.875 RCIM_Cubic>]}
           GreenCurve {keys [<at 0 0.525 RCIM_Cubic>,<at 0.5 0.6833>,<at 1 0.9208 RCIM_Cubic>]}
           RedCurve {keys [<at 0 1 RCIM_Cubic>,<at 0.5 1>,<at 1 1 RCIM_Cubic>]}
          }}))
       "Float (Alpha)"=&FloatFromCurve(FloatCurve=DataInterface() {propertyValues {Curve {keys [<at 0 1>,<at 1 0>]}}})
      ))
    )
   ) {isSetParameters=true}
   /SetVariables(
    Particles.Color_2:LinearColor=&MultiplyLinearColor(
     Linear Color=$Particles.Color_2
     Scale Factor=&MakeFloat4FromVectorAndFloat(
      "Vector (RGB)"=&VectorFromFloat(Value=&FloatFromCurve(FloatCurve=DataInterface() {propertyValues {Curve {keys [<at 0 1 RCIM_Cubic>, {arriveTangent=-1, interpMode=RCIM_Cubic, leaveTangent=-1, time=0.1, value=6}, <at 1 0 RCIM_Cubic>]}}}))
      "Float (Alpha)"=1
     )
    )
    Particles.Color_4:LinearColor=&MultiplyLinearColor(
     Linear Color=$Particles.Initial.Color_4
     Scale Factor=&MakeFloat4FromVectorAndFloat(
      "Vector (RGB)"=&VectorFromFloat(Value=&FloatFromCurve(FloatCurve=DataInterface() {propertyValues {Curve {keys [
            <at 0 100 RCIM_Cubic>
            {arriveTangent=-36.3579, interpMode=RCIM_Cubic, leaveTangent=-36.3595, tangentMode=RCTM_User, time=0.25, value=10}
            <at 1 1 RCIM_Cubic>
           ]}}}))
      "Float (Alpha)"=1
     )
    )
   ) {isSetParameters=true}
   &ScaleMeshSize(Scale Factor=&VectorFromCurve(VectorCurve=DataInterface() {propertyValues {
       XCurve {keys [
         {arriveTangent=2.9897, interpMode=RCIM_Cubic, leaveTangent=2.9897, tangentMode=RCTM_User, time=0, value=0}
         {arriveTangent=3.9553, interpMode=RCIM_Cubic, leaveTangent=3.9553, tangentMode=RCTM_User, time=0.2, value=0.6}
         <at 1 2 RCIM_Cubic>
        ]}
       YCurve {keys [
         {arriveTangent=14.017, interpMode=RCIM_Cubic, leaveTangent=14.017, tangentMode=RCTM_User, time=0, value=0}
         <at 0.2 1 RCIM_Cubic>
         <at 1 0 RCIM_Cubic>
        ]}
       ZCurve {keys [
         {arriveTangent=13.3462, interpMode=RCIM_Cubic, leaveTangent=13.3462, tangentMode=RCTM_User, time=0, value=0}
         <at 0.2 1 RCIM_Cubic>
         <at 1 0 RCIM_Cubic>
        ]}
      }}))
  ]
  rendererClasses [&NiagaraSpriteRendererProperties, &NiagaraMeshRendererProperties]
  renderers [
   {rendererClass=&NiagaraSpriteRendererProperties}
   {rendererClass=&NiagaraSpriteRendererProperties, rendererIndex=1}
   {rendererClass=&NiagaraSpriteRendererProperties, rendererIndex=2}
   {rendererClass=&NiagaraMeshRendererProperties, rendererIndex=3}
  ]
 }
}
summary {
 systemName=NS_Emote_CacaoLynx_Thrusters
 emitters [
  {emitterName=Thruster_Start, rendererClasses [&NiagaraSpriteRendererProperties, &NiagaraMeshRendererProperties]}
  {emitterName=Thruster_Loop, rendererClasses [&NiagaraMeshRendererProperties, &NiagaraSpriteRendererProperties]}
  {emitterName=Thruster_Smoke, rendererClasses [&NiagaraSpriteRendererProperties]}
  {bEnabled=false, emitterName=Comet_Flame_Mesh, rendererClasses [&NiagaraMeshRendererProperties]}
  {
   bEnabled=false
   emitterName=Thruster_Loop001
   rendererClasses [&NiagaraMeshRendererProperties, &NiagaraSpriteRendererProperties]
  }
  {emitterName=Minimal, rendererClasses [&NiagaraLightRendererProperties]}
 ]
 userVariables [
  User.Flame_Size {type=NiagaraFloat, defaultValue=2, description=""}
  User.ForwardDot {type=NiagaraFloat, defaultValue=1, description=""}
  User.Skeletal Mesh {
   type=NiagaraDataInterfaceSkeletalMesh
   defaultValue {
    struct=NiagaraExt_VariableValue_DataInterface
    value {dataInterface=&self:NiagaraDataInterfaceSkeletalMesh_1}
   }
  }
  User.Thruster_Height_Mult {type=NiagaraFloat, defaultValue=1, description=""}
 ]
}
